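# Additive unlit particle sprite material for a flickering flare with a hot, whitened center.
# Radial sphere-mask falloffs and a FlareMask texture are tinted by particle color, optionally multiplied by a panning rainbow gradient.
# Multiplied sines drive optional flicker, tuned by dynamic parameters or, for GPU particles, per-particle random, alongside depth fade (plain alpha on mobile).
# Optional "Use Distance Fade" offsets vertices away from object position, enlarging the flare with camera distance.
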Material /Figment_S08/DirtShell_8_5_1/Effects/Fort_Effects/Materials/Flares/M_Particle_Flare_HotCenter_Flicker
graph {
 properties {bUsedWithParticleSprites=true, blendMode=BLEND_Additive, pixelDepthOffsetMode=PDOM_Legacy, shadingModel=MSM_Unlit, usedShadingModels=Unlit}
 nodes [
  n1 = Constant() {r=0.5}
  n2 = SphereMask(A=TextureCoordinate(), B=@n1, Radius=@n1) {hardnessPercent=0}
  n3 = Power(Base=@n2) {constExponent=25}
  n4 = ComponentMask(None=TextureCoordinate() {uTiling=0, vTiling=0}) {g=true}
  n5 = Multiply(
   A=TextureSample(UVs=Panner(Coordinate=@n4) {speedX=1.6667}) {texture=/Game/Global/Textures/Default/Masks_and_Gradients/T_Grad_Rainbow}.RGB
   B=ParticleColor().RGB
  )
  n6 = ParticleColor()
  n7 = StaticSwitchParameter(
   True=MaterialFunctionCall(LDR Input=Clamp(None=@n5), HDR Tint=@n5) {materialFunction=&DeriveHDRfromLDR}.RemappedRange
   False=@n6.RGB
  ) {parameterName=Use Rainbow Color}
  n8 = Multiply(A=Power(Base=@n2) {constExponent=4}) {constB=0.15}
  n9 = TextureSampleParameter2D() {parameterName=FlareMask, texture=/Game/Effects/Fort_Effects/Textures/Flares/T_Flare_Spikeball}
  Add(A=Multiply(A=Multiply(A=@n3) {constB=2}, B=LinearInterpolate(B=@n7) {constA=1, constAlpha=0.7}), B=Multiply(A=Add(A=@n8, B=@n9.RGB), B=@n7))
  Constant() {r=1}
  n10 = ParticleRandom()
  n11 = DynamicParameter() {paramNames [FlickerFade,FlickerSpeed,Param3,Param4]}
  n12 = Multiply(
   A=Time()
   B=StaticSwitchParameter(True=LinearInterpolate(Alpha=@n10) {constA=1.2, constB=0.8}, False=@n11.FlickerSpeed) {parameterName=GPU Particles}
  )
  n13 = Sine(None=@n12) {period=0.075}
  n14 = Multiply(A=Multiply(A=Sine(None=@n12) {period=0.155}, B=Sine(None=@n12) {period=0.2}), B=Sine(None=@n12) {period=0.43})
  n15 = ScalarParameter() {defaultValue=0.4, desc="0 = full flicker amount", parameterName=GPUFlickerFadeMin}
  n16 = ScalarParameter() {defaultValue=1, desc="1 = no flicker", parameterName=GPUFlickerFadeMax, sortPriority=1}
  n17 = ParticleRandom()
  n18 = LinearInterpolate(
   A=Clamp(None=Multiply(A=@n13, B=@n14)) {maxDefault=3}
   Alpha=StaticSwitchParameter(True=LinearInterpolate(A=@n15, B=@n16, Alpha=@n17) {constA=0.5}, False=@n11.FlickerFade) {parameterName=GPU Particles}
  ) {constA=1}
  n19 = Constant() {r=0.8}
  FeatureLevelSwitch(Default=@n18, ES2=@n19, ES3_1=@n19, SM4_REMOVED=@n18)
  n20 = DepthFade(Opacity=@n6.A, FadeDistance=ScalarParameter() {defaultValue=30, parameterName=DepthFade})
  Multiply(
   A=StaticSwitchParameter(True=@n18, False=Constant() {r=1}) {defaultValue=true, parameterName=UseFlicker}
   B=FeatureLevelSwitch(Default=@n20, ES2=@n6.A, ES3_1=@n6.A, SM4_REMOVED=@n20, SM5=@n20, SM6=@n20)
  )
  n21 = ScalarParameter() {defaultValue=10, parameterName=MaxScaleIncrease}
  n22 = ObjectPositionWS()
  n23 = Multiply(A=Subtract(A=@n22, B=WorldPosition().XYZ)) {constB=-1}
  n24 = CameraPositionWS()
  n25 = Divide(
   A=Subtract(A=Distance(A=@n22, B=@n24), B=ScalarParameter() {defaultValue=2000, parameterName=DistanceOffset})
   B=ScalarParameter() {defaultValue=30000, parameterName=ScaleUpDistance}
  )
  StaticSwitchParameter(True=LinearInterpolate(B=Multiply(A=@n21, B=@n23) {constB=-1}, Alpha=Clamp(None=@n25)) {constB=0}, False=Constant()) {parameterName=Use Distance Fade}
 ]
}
